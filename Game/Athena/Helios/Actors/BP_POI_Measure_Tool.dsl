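; POI measure tool actor that previews a point of interest's footprint, label and streaming bounds in the level.
; Its construction script scales the bounds mesh from a size enum and sets text renders to the size label and POI name.
; It tints the bounds through a dynamic material from a color enum and optionally snaps to the Fort building grid.
; Streaming radius visibility and default-or-max scale are toggleable; gameplay events are empty.
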(event EventBeginPlay)

(event Collision|EventActorBeginOverlap
  (params (OtherActor :type Actor%20Object%20Reference)))

(event EventTick
  (params (DeltaSeconds :type Float%20%28single-precision%29)))

(fn GetPOISizeName
  (Variables|Default|SetPOI_Size_Text (Utilities|Map|Find (Utilities|Map|MakeMap "NewEnumerator0" "NSLOCTEXT(\"[D7BC4260D74E5BBFAEA6232AB673FFB6]\", \"40460C2448CEEF76BB8876985C7D781B\", \"Small\")" "NewEnumerator1" "NSLOCTEXT(\"[D7BC4260D74E5BBFAEA6232AB673FFB6]\", \"82DC766342B1E10F4FF6CEB9C837C44C\", \"Medium\")" "NewEnumerator2" "NSLOCTEXT(\"[D7BC4260D74E5BBFAEA6232AB673FFB6]\", \"9B46CDD64561AEF9BDC6E8815A64A2E3\", \"Large\")" "NewEnumerator3" "NSLOCTEXT(\"[D7BC4260D74E5BBFAEA6232AB673FFB6]\", \"8249B0C24C43511D2DCD2AAB423B71EE\", \"Xtra Large\")" "NewEnumerator4" "NSLOCTEXT(\"[D7BC4260D74E5BBFAEA6232AB673FFB6]\", \"8E9A7A324EF109CE2F256AB7961113C5\", \"Xtra Xtra Large\")" "NewEnumerator5" "NSLOCTEXT(\"[D7BC4260D74E5BBFAEA6232AB673FFB6]\", \"3329E0E44D813D25AADB65B61269EE5E\", \"Landmark\")" "NewEnumerator6" "NSLOCTEXT(\"[D7BC4260D74E5BBFAEA6232AB673FFB6]\", \"965042974A47A2998184E1802B789385\", \"Activity\")") (Variables|POI|GetPOI_Size))))

(fn SetPOIBoundsSize
  (switch Utilities|FlowControl|Switch|SwitchonE_POI_Measure_Tool_Size (Variables|POI|GetPOI_Size)
    (:NewEnumerator0
      (Variables|Default|SetPOI_Bounds_Scales "61.440000,61.440000,5.120000"))
    (:NewEnumerator1
      (Variables|Default|SetPOI_Bounds_Scales "92.160000,92.160000,5.120000"))
    (:NewEnumerator2
      (Variables|Default|SetPOI_Bounds_Scales "122.880000,122.880000,5.120000"))
    (:NewEnumerator3
      (Variables|Default|SetPOI_Bounds_Scales "179.200000,179.200000,5.120000"))
    (:NewEnumerator4
      (Variables|Default|SetPOI_Bounds_Scales "256.000000,256.000000,5.120000"))
    (:NewEnumerator5
      (Variables|Default|SetPOI_Bounds_Scales "25.600000,25.600000,5.120000"))
    (:NewEnumerator6
      (Variables|Default|SetPOI_Bounds_Scales "25.600000,25.600000,5.120000")))
  (Transformation|SetWorldScale3D (Variables|Default|GetPOIBounds) (Variables|Default|GetPOI_Bounds_Scales)))

(fn SetPOIColor
  (bind _returnvalue (Rendering|Material|CreateDynamicMaterialInstance (Variables|Default|GetPOIBounds) 0 "/Engine/BasicShapes/BasicShapeMaterial.BasicShapeMaterial" "None" :!class /Script/Engine.PrimitiveComponent))
  (Variables|Default|SetBoundsColorDMI _returnvalue)
  (switch Utilities|FlowControl|Switch|SwitchonE_POI_Measure_Tool_Color (Variables|POI|GetPOIColor)
    (:NewEnumerator0
      (Variables|Default|SetPOIColorValue "(R=1.000000,G=0.000000,B=0.000000,A=1.000000)"))
    (:NewEnumerator1
      (Variables|Default|SetPOIColorValue (Utilities|Struct|MakeLinearColor 0.4 0.1 0.0 1.0)))
    (:NewEnumerator2
      (Variables|Default|SetPOIColorValue "(R=1.000000,G=1.000000,B=0.000000,A=1.000000)"))
    (:NewEnumerator3
      (Variables|Default|SetPOIColorValue "(R=0.000000,G=1.000000,B=0.000000,A=1.000000)"))
    (:NewEnumerator4
      (Variables|Default|SetPOIColorValue "(R=0.000000,G=0.000000,B=1.000000,A=1.000000)"))
    (:NewEnumerator5
      (Variables|Default|SetPOIColorValue "(R=0.750000,G=0.000000,B=1.000000,A=1.000000)"))
    (:NewEnumerator7
      (Variables|Default|SetPOIColorValue "(R=0.900000,G=0.900000,B=0.900000,A=1.000000)")))
  (Rendering|Material|SetVectorParameterValue (Variables|Default|GetBoundsColorDMI) "Color" (Variables|Default|GetPOIColorValue) :!class /Script/Engine.MaterialInstanceDynamic))

(fn SnapToGrid
  (if (Variables|POI|GetSnapToBuildingGrid)
    (Transformation|SetActorLocation (Fortnite|SnaptoFortBuildingGrid (Transformation|GetActorLocation))))
  (return))

(fn ConstructionScript
  (CallFunction|SetPOIBoundsSize)
  (CallFunction|GetPOISizeName)
  (Rendering|Components|TextRender|SetText (Variables|Default|GetPOISize) (Variables|Default|GetPOI_Size_Text))
  (Rendering|Components|TextRender|SetText (Variables|Default|GetPOIName) (Variables|POI|GetPOI_Name_Text))
  (Rendering|SetVisibility (Variables|Default|GetStreamingRadius) (Variables|POI|GetShowStreamingBounds) false)
  (CallFunction|SetPOIColor)
  (CallFunction|SnaptoGrid)
  (if (Variables|POI|GetShowMaxGridStreamingBounds)
    (Transformation|SetWorldScale3D (Variables|Default|GetStreamingRadius) (Variables|Default|GetMaxStreamingRadius))
    (else
      (Transformation|SetWorldScale3D (Variables|Default|GetStreamingRadius) (Variables|Default|GetDefaultStreamingSize)))))
