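; Editor-style volume actor that bulk-sets whether building actors inside its box can be damaged.
; ExecuteCommand box-overlaps BuildingSMActors, handles props only if SetOnProps is on, and updates mismatches against bObjectsCanBeDamaged, recording undo copies and printing results.
; The prop branch leaves the SetCanbeDamaged value unwired, unlike the non-prop branch.
; Construction sizes the box, optionally in 512x512x384 building tiles, and offsets it upward by its Z extent.
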(event Custom|ExecuteCommand
  (bind _outactors (Collision|BoxOverlapActors (Transformation|GetActorLocation) (Components|Box|GetScaledBoxExtent (Variables|Default|GetIndestructibeVolume)) (Variables|Default|GetObject%20Types) "/Script/FortniteGame.BuildingSMActor" :!class /Script/Engine.KismetSystemLibrary))
  (for _array_element _outactors
    (bind _asbuilding_prop (Utilities|Casting|CastToBuildingProp _array_element)
      (:then
        (if (Variables|IndestructibleVolume|GetSetOnProps)
          (if (not (== (Variables|/Script/Engine.Actor|GetbCanBeDamaged _asbuilding_prop) (Variables|IndestructibleVolume|GetbObjectsCanBeDamaged)))
            (Fortnite|SetCanbeDamaged _asbuilding_prop)
            (Development|Editor|CreateCopyforUndoBuffer _asbuilding_prop)
            (if (Variables|/Script/Engine.Actor|GetbCanBeDamaged _asbuilding_prop)
              (Development|PrintString (Utilities|String|Append (Utilities|GetDisplayName _asbuilding_prop) ": CAN BE DAMAGE") false true "(R=1.000000,G=0.000000,B=0.000000,A=1.000000)" 0.0)
              (else
                (Development|PrintString (Utilities|String|Append (Utilities|GetDisplayName _asbuilding_prop) ": CANNOT BE DAMAGE") false true "(R=1.000000,G=0.000000,B=0.000000,A=1.000000)" 0.0))))))
      (:CastFailed
        (bind _asbuilding_static_mesh (Utilities|Casting|CastToBuildingSMActor _array_element))
        (if (not (== (Variables|/Script/Engine.Actor|GetbCanBeDamaged _asbuilding_static_mesh) (Variables|IndestructibleVolume|GetbObjectsCanBeDamaged)))
          (Fortnite|SetCanbeDamaged _asbuilding_static_mesh (Variables|IndestructibleVolume|GetbObjectsCanBeDamaged))
          (Development|Editor|CreateCopyforUndoBuffer _asbuilding_static_mesh)
          (if (Variables|/Script/Engine.Actor|GetbCanBeDamaged _asbuilding_static_mesh)
            (Development|PrintString (Utilities|String|Append (Utilities|GetDisplayName _asbuilding_static_mesh) ": CAN BE DAMAGE") false true "(R=1.000000,G=0.000000,B=0.000000,A=1.000000)" 0.0)
            (else
              (Development|PrintString (Utilities|String|Append (Utilities|GetDisplayName _asbuilding_static_mesh) ": CANNOT BE DAMAGE") false true "(R=1.000000,G=0.000000,B=0.000000,A=1.000000)" 0.0))))))))

(fn SetExtentsAndLocation
  (bind _indestructibevolume (Variables|Default|GetIndestructibeVolume))
  (bind (_boxextent_x _boxextent_y _boxextent_z) (Math|Vector|BreakVector (Variables|/Script/Engine.BoxComponent|GetBoxExtent_X (Variables|Default|GetIndestructibeVolume))))
  (bind (_volumesize_x _volumesize_y _volumesize_z) (Math|Vector|BreakVector (Variables|GetVolumeSize_X)))
  (if (Variables|IndestructibleVolume|GetbUseBuildingTileSizes)
    (Components|Box|SetBoxExtent _indestructibevolume (Utilities|Operators|Divide (Math|Vector|MakeVector (* _volumesize_x 512.0) (* _volumesize_y 512.0) (* _volumesize_z 384.0)) "2.000000,2.000000,2.000000") true)
    (elif (Utilities|Operators|Equal(==) (Variables|IndestructibleVolume|GetVolumeSize) "1.000000,1.000000,1.000000" 0)
      (Variables|IndestructibleVolume|SetVolumeSize "512.000000,512.000000,384.000000")
      (Components|Box|SetBoxExtent _indestructibevolume (/ (Variables|IndestructibleVolume|GetVolumeSize) "2.000000,2.000000,2.000000") true)
      (else
        _)))
  (Transformation|SetRelativeLocation (Variables|Default|GetIndestructibeVolume) (Math|Vector|MakeVector 0.0 0.0 _boxextent_z) false false))

(fn ConstructionScript
  (CallFunction|SetExtentsandLocation))
