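# M_POI_Blockout_Color is a Default Lit material, named and filed as a solid-color blockout material for Helios POI actors.
# It exposes a Color vector parameter, red by default, and a scalar Roughness parameter with a fairly matte default.
# The graph holds only these two parameters, with no textures or math, so material instances can vary tint and roughness.
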Material /Game/Athena/Helios/Actors/Materials/M_POI_Blockout_Color
graph {
 properties {pixelDepthOffsetMode=PDOM_Legacy, usedShadingModels=Default Lit}
 nodes [
  ScalarParameter() {defaultValue=0.6, parameterName=Roughness}
  VectorParameter() {parameterName=Color, defaultValue=<rgba 1 0 0 1>}
 ]
}
